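# Base AnimBlueprint for the drivable Quad (ATV), posing its moving parts procedurally from vehicle variables.
# A boost blend, switched by bIsBoosting, feeds bone transforms for axles, suspension, wheel steering and spin, and steering wheel.
# Paired Look At nodes aim shock halves at each other, then RoofSlot layers onto dyn_top_panel.
# The drive-shaft chain and placeholder-heavy PartsAnim state machine are unconnected to the output, and the EventGraph is empty.
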AnimBlueprint /Game/Athena/DrivableVehicles/Quad/Quad_Base/Meshes/Quad_Base_AnimBP_Base
target_skeleton=&/Quad_Base_Skeleton
graphs [
 AnimGraph {nodes [
   n1 = Look At - Bone: drive_shaft_base(Alpha=1) {
    alphaBoolBlend {blendOption=Linear}
    boneToModify {boneName=drive_shaft_base}
    lookAtLocation {x=0}
    lookAtTarget {boneReference {boneName=drive_shaft_rear}}
    lookAt_Axis {axis=<-1 0>}
    lookUp_Axis {axis {x=0, z=1}}
   }
   n2 = Get DriveShaftSpin() -> (DriveShaftSpin:Rotator)
   n3 = "Transform (Modify) Bone - Bone: drive_shaft"(ComponentPose=@n1, Alpha=1, Translation="0,0,0", Rotation=@n2, Scale="1,1,1") {rotationMode=BMM_Replace, rotationSpace=BCS_BoneSpace, alphaBoolBlend {blendOption=Linear}, boneToModify {boneName=drive_shaft}}
   "Transform (Modify) Bone - Bone: drive_shaft_rear"(ComponentPose=@n3, Alpha=1, Translation="0,0,0", Rotation=@n2, Scale="1,1,1") {rotationMode=BMM_Replace, rotationSpace=BCS_BoneSpace, alphaBoolBlend {blendOption=Linear}, boneToModify {boneName=drive_shaft_rear}}
   n4 = Blendspace Player 'ATV_Boost_BS'!(X=Get ChargingValue() -> (ChargingValue:float)) -> (Pose) {blendSpace=/Game/Animation/Game/MainPlayer/Vehicles/ATV/Parts/ATV_Boost_BS}
   n5 = Sequence Player 'ATV_Boost_Closed'() {sequence=/Game/Animation/Game/MainPlayer/Vehicles/ATV/Parts/ATV_Boost_Closed}
   n6 = Get bIsBoosting() -> (bIsBoosting:bool)
   n7 = Get AxlePositionF() -> (AxlePositionF:Vector)
   n8 = Get AxleRotationF() -> (AxleRotationF:Rotator)
   n9 = "Transform (Modify) Bone - Bone: axel_fr_l"(
    ComponentPose=Local To Component(LocalPose=Blend Poses by bool(BlendPose_0=@n5, BlendPose_1=@n4, BlendTime_1=0.3, bActiveValue=@n6) {blendType=Linear})
    Alpha=1
    Translation=@n7
    Rotation=@n8
    Scale="1,1,1"
   ) {
    rotationMode=BMM_Replace
    rotationSpace=BCS_BoneSpace
    translationMode=BMM_Replace
    translationSpace=BCS_BoneSpace
    alphaBoolBlend {blendOption=Linear}
    boneToModify {boneName=axel_fr_l}
   }
   n10 = "Transform (Modify) Bone - Bone: axel_fr_r"(ComponentPose=@n9, Alpha=1, Translation=@n7, Rotation=@n8, Scale="1,1,1") {
    rotationMode=BMM_Replace
    rotationSpace=BCS_BoneSpace
    translationMode=BMM_Replace
    translationSpace=BCS_BoneSpace
    alphaBoolBlend {blendOption=Linear}
    boneToModify {boneName=axel_fr_r}
   }
   n11 = "Transform (Modify) Bone - Bone: rear_suspension"(
    ComponentPose=@n10
    Alpha=1
    Translation=Get AxlePositionB() -> (AxlePositionB:Vector)
    Rotation=Get AxleRotationB() -> (AxleRotationB:Rotator)
    Scale="1,1,1"
   ) {
    rotationMode=BMM_Replace
    rotationSpace=BCS_BoneSpace
    translationMode=BMM_Replace
    translationSpace=BCS_BoneSpace
    alphaBoolBlend {blendOption=Linear}
    boneToModify {boneName=rear_suspension}
   }
   n12 = "Transform (Modify) Bone - Bone: wheel_tilted_fr_r"(ComponentPose=@n11, Alpha=1, Translation="0,0,0", Rotation=Get WheelRotationFR() -> (WheelRotationFR:Rotator), Scale="1,1,1") {
    rotationMode=BMM_Replace
    rotationSpace=BCS_BoneSpace
    translationSpace=BCS_BoneSpace
    alphaBoolBlend {blendOption=Linear}
    boneToModify {boneName=wheel_tilted_fr_r}
   }
   n13 = "Transform (Modify) Bone - Bone: wheel_steering_fr_l"(ComponentPose=@n12, Alpha=1, Translation="0,0,0", Rotation=Get WheelRotationFL() -> (WheelRotationFL:Rotator), Scale="1,1,1") {
    rotationMode=BMM_Replace
    rotationSpace=BCS_BoneSpace
    translationSpace=BCS_BoneSpace
    alphaBoolBlend {blendOption=Linear}
    boneToModify {boneName=wheel_steering_fr_l}
   }
   n14 = "Transform (Modify) Bone - Bone: wheel_tilted_fr_r"(ComponentPose=@n13, Alpha=1, Translation="0,0,0", Rotation=Get WheelSpinFR() -> (WheelSpinFR:Rotator), Scale="1,1,1") {rotationMode=BMM_Replace, rotationSpace=BCS_BoneSpace, alphaBoolBlend {blendOption=Linear}, boneToModify {boneName=wheel_tilted_fr_r}}
   n15 = "Transform (Modify) Bone - Bone: wheel_tilted_fr_l"(ComponentPose=@n14, Alpha=1, Translation="0,0,0", Rotation=Get WheelSpinFL() -> (WheelSpinFL:Rotator), Scale="1,1,1") {rotationMode=BMM_Replace, rotationSpace=BCS_BoneSpace, alphaBoolBlend {blendOption=Linear}, boneToModify {boneName=wheel_tilted_fr_l}}
   n16 = "Transform (Modify) Bone - Bone: wheel_tilted_bk_r"(ComponentPose=@n15, Alpha=1, Translation="0,0,0", Rotation=Get WheelSpinBR() -> (WheelSpinBR:Rotator), Scale="1,1,1") {rotationMode=BMM_Replace, rotationSpace=BCS_BoneSpace, alphaBoolBlend {blendOption=Linear}, boneToModify {boneName=wheel_tilted_bk_r}}
   n17 = "Transform (Modify) Bone - Bone: wheel_tilted_bk_l"(ComponentPose=@n16, Alpha=1, Translation="0,0,0", Rotation=Get WheelSpinBL() -> (WheelSpinBL:Rotator), Scale="1,1,1") {rotationMode=BMM_Replace, rotationSpace=BCS_BoneSpace, alphaBoolBlend {blendOption=Linear}, boneToModify {boneName=wheel_tilted_bk_l}}
   n18 = "Transform (Modify) Bone - Bone: steering_wheel"(ComponentPose=@n17, Alpha=1, Translation="0,0,0", Rotation=Get SteeringRotation() -> (SteeringRotation:Rotator), Scale="1,1,1") {rotationMode=BMM_Replace, rotationSpace=BCS_BoneSpace, alphaBoolBlend {blendOption=Linear}, boneToModify {boneName=steering_wheel}}
   n19 = Look At - Bone: shock_top_bk(ComponentPose=@n18, Alpha=1) {
    alphaBoolBlend {blendOption=Linear}
    boneToModify {boneName=shock_top_bk}
    lookAtLocation {x=0}
    lookAtTarget {boneReference {boneName=shock_bott_bk}}
    lookAt_Axis {axis {y=0, z=-1}}
    lookUp_Axis {axis=<0 -1>}
   }
   n20 = Look At - Bone: shock_bott_bk(ComponentPose=@n19, Alpha=1) {
    alphaBoolBlend {blendOption=Linear}
    boneToModify {boneName=shock_bott_bk}
    lookAtLocation {x=0}
    lookAtTarget {boneReference {boneName=shock_top_bk}}
    lookAt_Axis {axis {y=0, z=1}}
    lookUp_Axis {axis=<0 -1>}
   }
   n21 = Look At - Bone: shock_top_fr_l(ComponentPose=@n20, Alpha=1) {
    alphaBoolBlend {blendOption=Linear}
    boneToModify {boneName=shock_top_fr_l}
    lookAtLocation {x=0}
    lookAtTarget {boneReference {boneName=shock_bott_fr_l}}
    lookAt_Axis {axis {y=0, z=-1}}
    lookUp_Axis {axis=<0 -1>}
   }
   n22 = Look At - Bone: shock_bott_fr_l(ComponentPose=@n21, Alpha=1) {
    alphaBoolBlend {blendOption=Linear}
    boneToModify {boneName=shock_bott_fr_l}
    lookAtLocation {x=0}
    lookAtTarget {boneReference {boneName=shock_top_fr_l}}
    lookAt_Axis {axis {y=0, z=1}}
    lookUp_Axis {axis=<0 -1>}
   }
   n23 = Look At - Bone: shock_top_fr_r(ComponentPose=@n22, Alpha=1) {
    alphaBoolBlend {blendOption=Linear}
    boneToModify {boneName=shock_top_fr_r}
    lookAtLocation {x=0}
    lookAtTarget {boneReference {boneName=shock_bott_fr_r}}
    lookAt_Axis {axis {y=0, z=1}}
    lookUp_Axis {axis=<0 1>}
   }
   n24 = Look At - Bone: shock_bott_fr_r(ComponentPose=@n23, Alpha=1) {
    alphaBoolBlend {blendOption=Linear}
    boneToModify {boneName=shock_bott_fr_r}
    lookAtLocation {x=0}
    lookAtTarget {boneReference {boneName=shock_top_fr_r}}
    lookAt_Axis {axis {y=0, z=-1}}
    lookUp_Axis {axis=<0 -1>}
   }
   Output Pose(Result=Layered blend per bone(BasePose=Component To Local(ComponentPose=@n24), BlendPoses_0=Slot 'RoofSlot'() {slotName=RoofSlot}) {layerSetup [{branchFilters [{blendDepth=5, boneName=dyn_top_panel}]}]}) {layerGroup=None}
   AnimGraphNode_StateMachine_0 = PartsAnim() {is=title AnimGraphNode_StateMachine}
  ]}
 PartsAnim {
  entry=Default
  path=AnimGraph.AnimGraphNode_StateMachine_0.PartsAnim
  ids=id
  nodes [
   AnimStateNode_0 = Default() {is=title AnimStateNode}
   AnimStateNode_1 = Powersliding_Start() {is=title AnimStateNode}
   AnimStateNode_10 = PlaceHolder_Reverse_Start() {is=title AnimStateNode}
   AnimStateNode_11 = PlaceHolder_Reverse() {is=title AnimStateNode}
   AnimStateNode_12 = Boost_Loop() {is=title AnimStateNode}
   AnimStateNode_2 = Powersliding_Loop() {is=title AnimStateNode}
   AnimStateNode_3 = Powersliding_End() {is=title AnimStateNode}
   AnimStateNode_4 = Braking() {is=title AnimStateNode}
   AnimStateNode_5 = Boosting_Start() {is=title AnimStateNode}
   AnimStateNode_6 = PlaceHolder_Reverse_End() {is=title AnimStateNode}
   AnimStateNode_7 = PlaceHolder_Loop() {is=title AnimStateNode}
   AnimStateNode_8 = PlaceHolder_End() {is=title AnimStateNode}
   AnimStateNode_9 = PlaceHolder_Start() {is=title AnimStateNode}
   AnimStateTransitionNode_0 = AnimStateTransitionNode() {crossfadeDuration=0.1}
   AnimStateTransitionNode_1 = AnimStateTransitionNode() {bAutomaticRuleBasedOnSequencePlayerInState=true, crossfadeDuration=0.35}
   AnimStateTransitionNode_17 = AnimStateTransitionNode()
   AnimStateTransitionNode_18 = AnimStateTransitionNode() {bAutomaticRuleBasedOnSequencePlayerInState=true}
   AnimStateTransitionNode_19 = AnimStateTransitionNode()
   AnimStateTransitionNode_2 = AnimStateTransitionNode() {crossfadeDuration=0.1}
   AnimStateTransitionNode_20 = AnimStateTransitionNode() {bAutomaticRuleBasedOnSequencePlayerInState=true, crossfadeDuration=0.3}
   AnimStateTransitionNode_21 = AnimStateTransitionNode() {crossfadeDuration=0.35}
   AnimStateTransitionNode_22 = AnimStateTransitionNode() {bAutomaticRuleBasedOnSequencePlayerInState=true, crossfadeDuration=0.35}
   AnimStateTransitionNode_24 = AnimStateTransitionNode() {crossfadeDuration=0.1}
   AnimStateTransitionNode_25 = AnimStateTransitionNode()
   AnimStateTransitionNode_26 = AnimStateTransitionNode() {crossfadeDuration=0.1}
   AnimStateTransitionNode_27 = AnimStateTransitionNode() {bAutomaticRuleBasedOnSequencePlayerInState=true}
   AnimStateTransitionNode_28 = AnimStateTransitionNode()
   AnimStateTransitionNode_29 = AnimStateTransitionNode()
   AnimStateTransitionNode_3 = AnimStateTransitionNode()
   AnimStateTransitionNode_30 = AnimStateTransitionNode() {bAutomaticRuleBasedOnSequencePlayerInState=true}
   AnimStateTransitionNode_4 = AnimStateTransitionNode() {crossfadeDuration=0.1}
   AnimStateTransitionNode_5 = AnimStateTransitionNode() {bAutomaticRuleBasedOnSequencePlayerInState=true, crossfadeDuration=0.35}
   AnimStateTransitionNode_6 = AnimStateTransitionNode() {crossfadeDuration=0.1}
   AnimStateTransitionNode_7 = AnimStateTransitionNode() {bAutomaticRuleBasedOnSequencePlayerInState=true, crossfadeDuration=0.4}
   AnimStateTransitionNode_8 = AnimStateTransitionNode()
   AnimStateTransitionNode_9 = AnimStateTransitionNode() {crossfadeDuration=0.35}
  ]
  links [
   {
    dst=PlaceHolder_Loop
    rule=AnimGraph.AnimGraphNode_StateMachine_0.PartsAnim.AnimStateTransitionNode_9.Transition
    src=Boost_Loop
    via=AnimStateTransitionNode_9
   }
   {
    dst=Boost_Loop
    rule=AnimGraph.AnimGraphNode_StateMachine_0.PartsAnim.AnimStateTransitionNode_7.Transition
    src=Boosting_Start
    via=AnimStateTransitionNode_7
   }
   {
    dst=Powersliding_Start
    rule=AnimGraph.AnimGraphNode_StateMachine_0.PartsAnim.AnimStateTransitionNode_4.Transition
    src=Boosting_Start
    via=AnimStateTransitionNode_4
   }
   {
    dst=PlaceHolder_Loop
    rule=AnimGraph.AnimGraphNode_StateMachine_0.PartsAnim.AnimStateTransitionNode_5.Transition
    src=Braking
    via=AnimStateTransitionNode_5
   }
   {
    dst=Powersliding_Start
    rule=AnimGraph.AnimGraphNode_StateMachine_0.PartsAnim.AnimStateTransitionNode_6.Transition
    src=Braking
    via=AnimStateTransitionNode_6
   }
   {
    dst=PlaceHolder_Reverse_Start
    rule=AnimGraph.AnimGraphNode_StateMachine_0.PartsAnim.AnimStateTransitionNode_26.Transition
    src=Default
    via=AnimStateTransitionNode_26
   }
   {
    dst=PlaceHolder_Start
    rule=AnimGraph.AnimGraphNode_StateMachine_0.PartsAnim.AnimStateTransitionNode_21.Transition
    src=Default
    via=AnimStateTransitionNode_21
   }
   {
    dst=Default
    rule=AnimGraph.AnimGraphNode_StateMachine_0.PartsAnim.AnimStateTransitionNode_20.Transition
    src=PlaceHolder_End
    via=AnimStateTransitionNode_20
   }
   {
    dst=Boosting_Start
    rule=AnimGraph.AnimGraphNode_StateMachine_0.PartsAnim.AnimStateTransitionNode_8.Transition
    src=PlaceHolder_Loop
    via=AnimStateTransitionNode_8
   }
   {
    dst=Braking
    rule=AnimGraph.AnimGraphNode_StateMachine_0.PartsAnim.AnimStateTransitionNode_3.Transition
    src=PlaceHolder_Loop
    via=AnimStateTransitionNode_3
   }
   {
    dst=PlaceHolder_End
    rule=AnimGraph.AnimGraphNode_StateMachine_0.PartsAnim.AnimStateTransitionNode_19.Transition
    src=PlaceHolder_Loop
    via=AnimStateTransitionNode_19
   }
   {
    dst=PlaceHolder_Reverse_Start
    rule=AnimGraph.AnimGraphNode_StateMachine_0.PartsAnim.AnimStateTransitionNode_0.Transition
    src=PlaceHolder_Loop
    via=AnimStateTransitionNode_0
   }
   {
    dst=Powersliding_Start
    rule=AnimGraph.AnimGraphNode_StateMachine_0.PartsAnim.AnimStateTransitionNode_17.Transition
    src=PlaceHolder_Loop
    via=AnimStateTransitionNode_17
   }
   {
    dst=PlaceHolder_Reverse_End
    rule=AnimGraph.AnimGraphNode_StateMachine_0.PartsAnim.AnimStateTransitionNode_29.Transition
    src=PlaceHolder_Reverse
    via=AnimStateTransitionNode_29
   }
   {
    dst=PlaceHolder_Start
    rule=AnimGraph.AnimGraphNode_StateMachine_0.PartsAnim.AnimStateTransitionNode_28.Transition
    src=PlaceHolder_Reverse
    via=AnimStateTransitionNode_28
   }
   {
    dst=Powersliding_Start
    rule=AnimGraph.AnimGraphNode_StateMachine_0.PartsAnim.AnimStateTransitionNode_24.Transition
    src=PlaceHolder_Reverse
    via=AnimStateTransitionNode_24
   }
   {
    dst=Default
    rule=AnimGraph.AnimGraphNode_StateMachine_0.PartsAnim.AnimStateTransitionNode_30.Transition
    src=PlaceHolder_Reverse_End
    via=AnimStateTransitionNode_30
   }
   {
    dst=PlaceHolder_Reverse
    rule=AnimGraph.AnimGraphNode_StateMachine_0.PartsAnim.AnimStateTransitionNode_27.Transition
    src=PlaceHolder_Reverse_Start
    via=AnimStateTransitionNode_27
   }
   {
    dst=Powersliding_Start
    rule=AnimGraph.AnimGraphNode_StateMachine_0.PartsAnim.AnimStateTransitionNode_25.Transition
    src=PlaceHolder_Reverse_Start
    via=AnimStateTransitionNode_25
   }
   {
    dst=PlaceHolder_Loop
    rule=AnimGraph.AnimGraphNode_StateMachine_0.PartsAnim.AnimStateTransitionNode_22.Transition
    src=PlaceHolder_Start
    via=AnimStateTransitionNode_22
   }
   {
    dst=PlaceHolder_Loop
    rule=AnimGraph.AnimGraphNode_StateMachine_0.PartsAnim.AnimStateTransitionNode_18.Transition
    src=Powersliding_End
    via=AnimStateTransitionNode_18
   }
   {
    dst=Powersliding_End
    rule=AnimGraph.AnimGraphNode_StateMachine_0.PartsAnim.AnimStateTransitionNode_2.Transition
    src=Powersliding_Loop
    via=AnimStateTransitionNode_2
   }
   {
    dst=Powersliding_Loop
    rule=AnimGraph.AnimGraphNode_StateMachine_0.PartsAnim.AnimStateTransitionNode_1.Transition
    src=Powersliding_Start
    via=AnimStateTransitionNode_1
   }
  ]
 }
 Default {path=AnimGraph.AnimGraphNode_StateMachine_0.PartsAnim.AnimStateNode_0.Default, nodes [Output Animation Pose(Result=Sequence Player())]}
 Powersliding_Start {
  path=AnimGraph.AnimGraphNode_StateMachine_0.PartsAnim.AnimStateNode_1.Powersliding_Start
  nodes [Output Animation Pose(Result=Sequence Player() {bLoopAnimation=false})]
 }
 PlaceHolder_Reverse_Start {
  path=AnimGraph.AnimGraphNode_StateMachine_0.PartsAnim.AnimStateNode_10.PlaceHolder_Reverse_Start
  nodes [Output Animation Pose(Result=Layered blend per bone(BasePose=Sequence Player() {groupName=Start, method=SyncGroup}, BlendPoses_0=Sequence Player() {groupName=Start, method=SyncGroup}) {bMeshSpaceRotationBlend=true, layerSetup [{branchFilters [{boneName=pedal_gas}]}]})]
 }
 PlaceHolder_Reverse {
  path=AnimGraph.AnimGraphNode_StateMachine_0.PartsAnim.AnimStateNode_11.PlaceHolder_Reverse
  nodes [Output Animation Pose(Result=Layered blend per bone(BasePose=Sequence Player(), BlendPoses_0=Sequence Player()) {bMeshSpaceRotationBlend=true, layerSetup [{branchFilters [{boneName=pedal_gas}]}]})]
 }
 Boost_Loop {path=AnimGraph.AnimGraphNode_StateMachine_0.PartsAnim.AnimStateNode_12.Boost_Loop, nodes [Output Animation Pose(Result=Sequence Player())]}
 Powersliding_Loop {
  path=AnimGraph.AnimGraphNode_StateMachine_0.PartsAnim.AnimStateNode_2.Powersliding_Loop
  nodes [Output Animation Pose(Result=Sequence Player())]
 }
 Powersliding_End {
  path=AnimGraph.AnimGraphNode_StateMachine_0.PartsAnim.AnimStateNode_3.Powersliding_End
  nodes [Output Animation Pose(Result=Sequence Player() {bLoopAnimation=false})]
 }
 Braking {
  path=AnimGraph.AnimGraphNode_StateMachine_0.PartsAnim.AnimStateNode_4.Braking
  nodes [Output Animation Pose(Result=Layered blend per bone(BasePose=Sequence Player() {groupName=Braking, method=SyncGroup}, BlendPoses_0=Sequence Player() {groupName=Braking, method=SyncGroup}) {bMeshSpaceRotationBlend=true, layerSetup [{branchFilters [{boneName=pedal_brake}]}]})]
 }
 Boosting_Start {
  path=AnimGraph.AnimGraphNode_StateMachine_0.PartsAnim.AnimStateNode_5.Boosting_Start
  nodes [Output Animation Pose(Result=Sequence Player() {bLoopAnimation=false})]
 }
 PlaceHolder_Reverse_End {
  path=AnimGraph.AnimGraphNode_StateMachine_0.PartsAnim.AnimStateNode_6.PlaceHolder_Reverse_End
  nodes [Output Animation Pose(Result=Layered blend per bone(
     BasePose=Sequence Player() {bLoopAnimation=false, groupName=BackToIdle, method=SyncGroup}
     BlendPoses_0=Sequence Player() {bLoopAnimation=false, groupName=BackToIdle, method=SyncGroup}
    ) {bMeshSpaceRotationBlend=true, layerSetup [{branchFilters [{boneName=pedal_gas}]}]})]
 }
 PlaceHolder_Loop {
  path=AnimGraph.AnimGraphNode_StateMachine_0.PartsAnim.AnimStateNode_7.PlaceHolder_Loop
  nodes [Output Animation Pose(Result=Layered blend per bone(BasePose=Sequence Player(), BlendPoses_0=Sequence Player()) {bMeshSpaceRotationBlend=true, layerSetup [{branchFilters [{boneName=pedal_gas}]}]})]
 }
 PlaceHolder_End {
  path=AnimGraph.AnimGraphNode_StateMachine_0.PartsAnim.AnimStateNode_8.PlaceHolder_End
  nodes [Output Animation Pose(Result=Layered blend per bone(
     BasePose=Sequence Player() {bLoopAnimation=false, groupName=BackToIdle, method=SyncGroup}
     BlendPoses_0=Sequence Player() {bLoopAnimation=false, groupName=BackToIdle, method=SyncGroup}
    ) {bMeshSpaceRotationBlend=true, layerSetup [{branchFilters [{boneName=pedal_gas}]}]})]
 }
 PlaceHolder_Start {
  path=AnimGraph.AnimGraphNode_StateMachine_0.PartsAnim.AnimStateNode_9.PlaceHolder_Start
  nodes [Output Animation Pose(Result=Layered blend per bone(BasePose=Sequence Player() {groupName=Start, method=SyncGroup}, BlendPoses_0=Sequence Player() {groupName=Start, method=SyncGroup}) {bMeshSpaceRotationBlend=true, layerSetup [{branchFilters [{boneName=pedal_gas}]}]})]
 }
 Transition {
  path=AnimGraph.AnimGraphNode_StateMachine_0.PartsAnim.AnimStateTransitionNode_0.Transition
  nodes [Result(bCanEnterTransition=Get bReverseSpeedIsGreaterThanOne() -> (bReverseSpeedIsGreaterThanOne:bool))]
 }
 Transition {path=AnimGraph.AnimGraphNode_StateMachine_0.PartsAnim.AnimStateTransitionNode_1.Transition, nodes [Result()]}
 Transition {
  path=AnimGraph.AnimGraphNode_StateMachine_0.PartsAnim.AnimStateTransitionNode_17.Transition
  nodes [Result(bCanEnterTransition=Get bSpeedIsGreaterThanFiveAndPlayerHitSpace() -> (bSpeedIsGreaterThanFiveAndPlayerHitSpace:bool))]
 }
 Transition {path=AnimGraph.AnimGraphNode_StateMachine_0.PartsAnim.AnimStateTransitionNode_18.Transition, nodes [Result()]}
 Transition {
  path=AnimGraph.AnimGraphNode_StateMachine_0.PartsAnim.AnimStateTransitionNode_19.Transition
  nodes [Result(bCanEnterTransition=Get bForwardSpeedIsNearlyZero() -> (bForwardSpeedIsNearlyZero:bool))]
 }
 Transition {
  path=AnimGraph.AnimGraphNode_StateMachine_0.PartsAnim.AnimStateTransitionNode_2.Transition
  nodes [Result(bCanEnterTransition=NOT Boolean(A=Get bPlayerHitSpaceBar() -> (bPlayerHitSpaceBar:bool)))]
 }
 Transition {path=AnimGraph.AnimGraphNode_StateMachine_0.PartsAnim.AnimStateTransitionNode_20.Transition, nodes [Result()]}
 Transition {
  path=AnimGraph.AnimGraphNode_StateMachine_0.PartsAnim.AnimStateTransitionNode_21.Transition
  nodes [Result(bCanEnterTransition="float > float"(A=Get VehicleSpeed() -> (VehicleSpeed:float), B=1) {is=title K2Node_CallFunction})]
 }
 Transition {path=AnimGraph.AnimGraphNode_StateMachine_0.PartsAnim.AnimStateTransitionNode_22.Transition, nodes [Result()]}
 Transition {
  path=AnimGraph.AnimGraphNode_StateMachine_0.PartsAnim.AnimStateTransitionNode_24.Transition
  nodes [Result(bCanEnterTransition=Get bPlayerHitSpaceBar())]
 }
 Transition {
  path=AnimGraph.AnimGraphNode_StateMachine_0.PartsAnim.AnimStateTransitionNode_25.Transition
  nodes [Result(bCanEnterTransition=Get bPlayerHitSpaceBar())]
 }
 Transition {
  path=AnimGraph.AnimGraphNode_StateMachine_0.PartsAnim.AnimStateTransitionNode_26.Transition
  nodes [Result(bCanEnterTransition=Get bReverseSpeedIsGreaterThanOne())]
 }
 Transition {path=AnimGraph.AnimGraphNode_StateMachine_0.PartsAnim.AnimStateTransitionNode_27.Transition, nodes [Result()]}
 Transition {
  path=AnimGraph.AnimGraphNode_StateMachine_0.PartsAnim.AnimStateTransitionNode_28.Transition
  nodes [Result(bCanEnterTransition=Get bForwardSpeedIsGreaterThanOne() -> (bForwardSpeedIsGreaterThanOne:bool))]
 }
 Transition {
  path=AnimGraph.AnimGraphNode_StateMachine_0.PartsAnim.AnimStateTransitionNode_29.Transition
  nodes [Result(bCanEnterTransition=Get bForwardSpeedIsNearlyZero())]
 }
 Transition {
  path=AnimGraph.AnimGraphNode_StateMachine_0.PartsAnim.AnimStateTransitionNode_3.Transition
  nodes [Result(bCanEnterTransition=Get bIsBraking() -> (bIsBraking:bool))]
 }
 Transition {path=AnimGraph.AnimGraphNode_StateMachine_0.PartsAnim.AnimStateTransitionNode_30.Transition, nodes [Result()]}
 Transition {
  path=AnimGraph.AnimGraphNode_StateMachine_0.PartsAnim.AnimStateTransitionNode_4.Transition
  nodes [Result(bCanEnterTransition=Get bIsPowerSliding() -> (bIsPowerSliding:bool))]
 }
 Transition {path=AnimGraph.AnimGraphNode_StateMachine_0.PartsAnim.AnimStateTransitionNode_5.Transition, nodes [Result()]}
 Transition {
  path=AnimGraph.AnimGraphNode_StateMachine_0.PartsAnim.AnimStateTransitionNode_6.Transition
  nodes [Result(bCanEnterTransition=Get bSpeedIsGreaterThanFiveAndPlayerHitSpace())]
 }
 Transition {path=AnimGraph.AnimGraphNode_StateMachine_0.PartsAnim.AnimStateTransitionNode_7.Transition, nodes [Result()]}
 Transition {
  path=AnimGraph.AnimGraphNode_StateMachine_0.PartsAnim.AnimStateTransitionNode_8.Transition
  nodes [Result(bCanEnterTransition=Get bIsBoosting())]
 }
 Transition {
  path=AnimGraph.AnimGraphNode_StateMachine_0.PartsAnim.AnimStateTransitionNode_9.Transition
  nodes [Result(bCanEnterTransition=NOT Boolean(A=Get bIsBoosting()))]
 }
 EventGraph {links [], nodes []}
]
variables []
