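# Physics asset for the drivable Quad's base mesh, covering small parts that appear set up for secondary jiggle motion.
# Each bone gets one capsule body.
# Mirror and mud guards constrain to the steering wheel; exhausts, bat and its wires, headlight and gauges constrain to the frame.
# The steering wheel-body-frame chain has free angular motion, while attached parts carry swing limit angles with tighter twist limits.
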PhysicsAsset /Game/Athena/DrivableVehicles/Quad/Quad_Base/Meshes/Quad_Base_Dynamics
bodies {
 bat [{shapeType=Capsule, length=13.9052, radius=9.2456, center=<-6 8.5 -6.2>}]
 bat_wire_01 [{shapeType=Capsule, length=3, radius=2, center=<3 0 0>}]
 bat_wire_02 [{shapeType=Capsule, length=3, radius=2, center=<3 0 0>}]
 bat_wire_03 [{shapeType=Capsule, length=3, radius=2, center=<3 0 0>}]
 body [{shapeType=Capsule, length=10, radius=10}]
 dyn_mirror [{shapeType=Capsule, length=12.081, radius=6.5925, center=<3.6561 -8.7547 6.5091>, rotation=<rot 0 -1.2002 -51.7681>}]
 dyn_mud_guard_l [{shapeType=Capsule, length=9.331, radius=4.3425, center=<0 -6.3249 1.7689>, rotation=<rot 0 0 105.0004>}]
 dyn_mud_guard_r [{shapeType=Capsule, length=9.331, radius=4.3425, center=<0 6.3249 1.7689>, rotation=<rot 0 0 -105.0004>}]
 exhaust_l [{shapeType=Capsule, length=34.3403, radius=9.1329, center=<-16.9307 -12.2615 -0.1453>, rotation=<rot 0 -77.5442 -65.9944>}]
 exhaust_r [{shapeType=Capsule, length=34.3403, radius=9.1329, center=<-16.0723 12.0719 -0.5368>, rotation=<rot 0 -102.4556 -65.9943>}]
 frame [{shapeType=Capsule, length=5, radius=5}]
 guage_big [{shapeType=Capsule, length=3.5, radius=6.5, center=<-5.0602 -0.2217 -4.9448>, rotation=<rot -34.9998 0 5.0002>}]
 guage_sm [{shapeType=Capsule, length=0, radius=3.75, center=<-1.0319 -1.1304 -1.268>, rotation=<rot -34.9998 0 5.0003>}]
 headlight_r [{shapeType=Capsule, length=2.5, radius=7.5, center=<-4.571 0.5369 2.6958>, rotation=<rot 0 0 5.0002>}]
 steering_wheel [{shapeType=Capsule, length=5, radius=5, rotation=<rot 0 0 90>}]
}
constraints [
 {bone1Name=dyn_mirror, bone2Name=steering_wheel, swing1LimitDegrees=45, swing2LimitDegrees=45, twistLimitDegrees=5}
 {
  bone1Name=steering_wheel
  bone2Name=body
  swing1LimitDegrees=45
  swing1Motion=Free
  swing2LimitDegrees=45
  swing2Motion=Free
  twistLimitDegrees=45
  twistMotion=Free
 }
 {bone1Name=dyn_mud_guard_l, bone2Name=steering_wheel, swing1LimitDegrees=45, swing2LimitDegrees=45, twistLimitDegrees=5}
 {bone1Name=dyn_mud_guard_r, bone2Name=steering_wheel, swing1LimitDegrees=45, swing2LimitDegrees=45, twistLimitDegrees=5}
 {
  bone1Name=body
  bone2Name=frame
  swing1LimitDegrees=45
  swing1Motion=Free
  swing2LimitDegrees=45
  swing2Motion=Free
  twistLimitDegrees=45
  twistMotion=Free
 }
 {bone1Name=exhaust_l, bone2Name=frame, swing1LimitDegrees=45, swing2LimitDegrees=45, twistLimitDegrees=10}
 {bone1Name=exhaust_r, bone2Name=frame, swing1LimitDegrees=45, swing2LimitDegrees=45, twistLimitDegrees=10}
 {bone1Name=bat, bone2Name=frame, swing1LimitDegrees=60, swing2LimitDegrees=60, twistLimitDegrees=10}
 {bone1Name=bat_wire_01, bone2Name=frame, swing1LimitDegrees=60, swing2LimitDegrees=60, twistLimitDegrees=10}
 {bone1Name=bat_wire_02, bone2Name=frame, swing1LimitDegrees=60, swing2LimitDegrees=60, twistLimitDegrees=10}
 {bone1Name=bat_wire_03, bone2Name=frame, swing1LimitDegrees=60, swing2LimitDegrees=60, twistLimitDegrees=10}
 {bone1Name=headlight_r, bone2Name=frame, swing1LimitDegrees=45, swing2LimitDegrees=45, twistLimitDegrees=5}
 {bone1Name=guage_big, bone2Name=frame, swing1LimitDegrees=45, swing2LimitDegrees=45, twistLimitDegrees=5}
 {bone1Name=guage_sm, bone2Name=frame, swing1LimitDegrees=45, swing2LimitDegrees=45, twistLimitDegrees=5}
]
