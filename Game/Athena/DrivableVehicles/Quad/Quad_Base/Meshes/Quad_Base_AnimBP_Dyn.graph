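# Secondary-motion Anim Blueprint for the drivable Quad, applying dynamics and procedural bone offsets to an input pose.
# A RigidBody node simulates the Quad_Base_Dynamics asset, then sprung AnimDynamics on each exhaust feed additive CurrentDriveJiggle rotation weighted by SpeedABS.
# The cowcatcher bone then slides along its local X with interpolated alpha while bIsBoosting.
# Its EventGraph and variable list are empty, and all these skeletal-control nodes use LOD threshold 1.
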AnimBlueprint /Game/Athena/DrivableVehicles/Quad/Quad_Base/Meshes/Quad_Base_AnimBP_Dyn
target_skeleton=&/Quad_Base_Skeleton
graphs [
 AnimGraph {nodes [
   AnimGraphNode_SubInput_0 = Input Pose()
   n1 = RigidBody(ComponentPose=Local To Component(LocalPose=@AnimGraphNode_SubInput_0)) {
    bOverrideWorldGravity=true
    lODThreshold=1
    overridePhysicsAsset=&/Quad_Base_Dynamics
    alphaBoolBlend {blendOption=Linear}
    componentAppliedLinearAccClamp=<1000 1000 1000>
    componentLinearAccScale=<5 5 5>
    componentLinearVelScale=<1 1 1>
    overrideWorldGravity {z=-980}
   }
   n2 = Anim Dynamics - Bone: exhaust_l(ComponentPose=@n1, Alpha=1, ExternalForce="0,0,0") {
    angularDampingOverride=0.99
    angularSpringConstant=70000
    bAngularSpring=true
    bGravityOverrideInSimSpace=true
    bOverrideAngularDamping=true
    lODThreshold=1
    numSolverIterationsPostUpdate=4
    numSolverIterationsPreUpdate=16
    alphaBoolBlend {blendOption=Linear}
    boundBone {boneName=exhaust_l}
    componentAppliedLinearAccClamp=<1000 1000 1000>
    componentLinearAccScale=<1 1 1>
    componentLinearVelScale=<1 1 1>
    physicsBodyDefinitions [{
      sphereCollisionRadius=0
      boundBone {boneName=exhaust_l}
      boxExtents=<30 15 10>
      constraintSetup {angularLimitsMax=<0 1 6>, angularLimitsMin=<0 -10 -1>, angularTarget=<1 0 0>}
      localJointOffset=<-1 0 0>
     }]
   }
   n3 = Get CurrentDriveJiggle() -> (CurrentDriveJiggle:Rotator)
   n4 = Get SpeedABS() -> (SpeedABS:float)
   n5 = "Transform (Modify) Bone - Bone: exhaust_l"(ComponentPose=@n2, Alpha=@n4, Translation="0,0,0", Rotation=@n3, Scale="1,1,1") {
    lODThreshold=1
    rotationMode=BMM_Additive
    rotationSpace=BCS_BoneSpace
    alphaBoolBlend {blendOption=Linear}
    alphaScaleBiasClamp {bMapRange=true, inRange {max=35}}
    boneToModify {boneName=exhaust_l}
   }
   n6 = Anim Dynamics - Bone: exhaust_r(ComponentPose=@n5, Alpha=1, ExternalForce="0,0,0") {
    angularDampingOverride=0.99
    angularSpringConstant=70000
    bAngularSpring=true
    bGravityOverrideInSimSpace=true
    bOverrideAngularDamping=true
    lODThreshold=1
    numSolverIterationsPostUpdate=4
    numSolverIterationsPreUpdate=16
    alphaBoolBlend {blendOption=Linear}
    boundBone {boneName=exhaust_r}
    componentAppliedLinearAccClamp=<1000 1000 1000>
    componentLinearAccScale=<1 1 1>
    componentLinearVelScale=<1 1 1>
    physicsBodyDefinitions [{
      sphereCollisionRadius=0
      boundBone {boneName=exhaust_r}
      boxExtents=<30 15 10>
      constraintSetup {angularLimitsMax=<0 1 6>, angularLimitsMin=<0 -10 -1>, angularTarget=<1 0 0>}
      localJointOffset=<1 0 0>
     }]
   }
   n7 = "Transform (Modify) Bone - Bone: exhaust_r"(ComponentPose=@n6, Alpha=@n4, Translation="0,0,0", Rotation=@n3, Scale="1,1,1") {
    lODThreshold=1
    rotationMode=BMM_Additive
    rotationSpace=BCS_BoneSpace
    alphaBoolBlend {blendOption=Linear}
    alphaScaleBiasClamp {bMapRange=true, inRange {max=35}}
    boneToModify {boneName=exhaust_r}
   }
   n8 = "Transform (Modify) Bone - Bone: cowcatcher"(
    ComponentPose=@n7
    Alpha="To Float (Boolean)"(InBool=Get bIsBoosting() -> (bIsBoosting:bool))
    Translation="18,0.000,0.000"
    Rotation="0,0,0"
    Scale="1,1,1"
   ) {
    lODThreshold=1
    rotationSpace=BCS_BoneSpace
    translationMode=BMM_Additive
    alphaBoolBlend {blendOption=Linear}
    alphaScaleBiasClamp {bInterpResult=true}
    boneToModify {boneName=cowcatcher}
    translation {x=18}
   }
   Output Pose(Result=Component To Local(ComponentPose=@n8)) {layerGroup=None}
  ]}
 EventGraph {links [], nodes []}
]
variables []
