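# Physics asset for the drivable Quad vehicle's base mesh in Athena.
# It holds a single body on the frame bone made of two capsules, one named for driver collision and one for passenger collision.
# The passenger capsule sits behind and above the driver's, and no constraints are defined.
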PhysicsAsset /Game/Athena/DrivableVehicles/Quad/Quad_Base/Meshes/Quad_Base_Physics
bodies {frame [
  {shapeType=Capsule, length=60.9193, radius=30.7642, shapeName=PassengerCollision, center=<-75.9351 0.0001 99.1124>}
  {shapeType=Capsule, length=40.8247, radius=30.7408, shapeName=DriverCollision, center=<-1.5408 -0.0001 87.6799>}
 ]}
constraints []
